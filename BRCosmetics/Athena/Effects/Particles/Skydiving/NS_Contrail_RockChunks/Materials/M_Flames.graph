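# Unlit flame material for Niagara mesh particles in the rock chunks skydiving contrail.
# Emissive blends two tinted intensities by a product of masks, built from invertible Fresnel, invertible vertex-red gradient and particle life from a dynamic parameter.
# Each mask has an influence control, and named reroutes carry the intermediate results.
# An optional world position offset pushes vertices along normals using panning blurred noise masked by vertex red, then remaps it.
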Material /BRCosmetics/Athena/Effects/Particles/Skydiving/NS_Contrail_RockChunks/Materials/M_Flames
graph {
 properties {bUsedWithNiagaraMeshParticles=true, pixelDepthOffsetMode=PDOM_Legacy, shadingModel=MSM_Unlit, usedShadingModels=Unlit}
 nodes [
  n1 = Fresnel(ExponentIn=ScalarParameter() {defaultValue=1, group=ColorFresnel, parameterName=Fresnel_Pow})
  n2 = Round(None=ScalarParameter() {group=ColorFresnel, parameterName=Fresnel_Invert?, sliderMax=1})
  n3 = LinearInterpolate(
   A=Constant() {r=1}
   B=LinearInterpolate(A=@n1, B=OneMinus(None=@n1), Alpha=Saturate(None=@n2))
   Alpha=ScalarParameter() {defaultValue=1, group=ColorFresnel, parameterName=FresnelInfluence, sliderMax=1, sortPriority=0}
  ) {constA=1}
  n4 = NamedRerouteUsage()
  NamedRerouteDeclaration(None=LinearInterpolate(A=@n3, B=Multiply(A=@n3, B=@n4), Alpha=ScalarParameter() {group=ColorFresnel, parameterName=FresnelParticleLifeMultiply})) {name=Fresnel, nodeColor=<rgba 0 1 0.3906 1>}
  n5 = VertexColor()
  n6 = ScalarParameter() {defaultValue=1, group=Vert Color Gradient, parameterName=VertColGradientScale}
  n7 = Saturate(None=Power(Base=Multiply(A=@n5.R, B=@n6), Exp=ScalarParameter() {defaultValue=1, group=Vert Color Gradient, parameterName=VertColGradientPow}))
  n8 = OneMinus(None=@n7)
  n9 = Round(None=ScalarParameter() {group=Vert Color Gradient, parameterName=InvertVertColGradient?, sliderMax=1})
  NamedRerouteDeclaration(None=LinearInterpolate(
    A=Constant() {r=1}
    B=LinearInterpolate(A=@n7, B=@n8, Alpha=@n9)
    Alpha=ScalarParameter() {defaultValue=1, group=Vert Color Gradient, parameterName=VertColGradientInfluence, sliderMax=1, sortPriority=0}
   ) {constA=1}) {name=VertColGradient, nodeColor=<rgba 1 0 0.9141 1>}
  n10 = VertexNormalWS()
  n11 = Multiply(A=TextureCoordinate(), B=ScalarParameter() {defaultValue=1, group=WPO, parameterName=WPO_NoiseScale})
  n12 = ScalarParameter() {defaultValue=0.5, group=WPO, parameterName=WPO_Speed}
  n13 = Multiply(
   A=TextureSample(
    UVs=Panner(Coordinate=@n11, Speed=@n12)
    Tex=TextureObject() {samplerType=SAMPLERTYPE_Grayscale, texture=&LowResBlurredNoise}
   ) {samplerType=SAMPLERTYPE_Grayscale, texture=&LowResBlurredNoise}.R
   B=VertexColor().R
  )
  MaterialExpressionRemap_1 = MaterialXRemap!(
   None=Multiply(A=Multiply(A=@n10, B=@n13), B=ScalarParameter() {defaultValue=1, group=WPO, parameterName=WPO_Intensity})
   InputLow
   InputHigh
   TargetLow
   TargetHigh
  ) {desc=CenterOffset, targetLowDefault=-1}
  NamedRerouteDeclaration(None=StaticSwitch(True=@MaterialExpressionRemap_1, False=Constant(), Value=StaticBoolParameter() {group=WPO, parameterName=WPO?})) {name=WPOFINAL, nodeColor=<rgba 0 1 0.9766 1>}
  n14 = VectorParameter() {group=Color, parameterName=EmissiveColor01, defaultValue=<rgba 1 0.5 0 1>}
  n15 = ScalarParameter() {defaultValue=1, group=Color, parameterName=EmissiveIntensity01}
  n16 = VectorParameter() {group=Color, parameterName=EmissiveColor02, defaultValue=<rgba 1 0 0 1>}
  n17 = ScalarParameter() {defaultValue=1, group=Color, parameterName=EmissiveIntensity02}
  n18 = Multiply(A=Multiply(A=NamedRerouteUsage(), B=NamedRerouteUsage()), B=NamedRerouteUsage())
  NamedRerouteDeclaration(None=LinearInterpolate(A=Multiply(A=@n14.RGB, B=@n15), B=Multiply(A=@n16.RGB, B=@n17), Alpha=Saturate(None=@n18))) {name=EmissiveFINAL, nodeColor=<rgba 1 0.7734 0 1>}
  NamedRerouteDeclaration(None=LinearInterpolate(
    A=Constant() {r=1}
    B=DynamicParameter() {paramNames [ColorGradientLife, Param2, Param3, Param4]}.ColorGradientLife
    Alpha=ScalarParameter() {defaultValue=1, group=Particles, parameterName=ParticleLifeInfluence, sliderMax=1}
   ) {constA=1}) {name=ParticleLife, nodeColor=<rgba 0 0.2031 1 1>}
  NamedRerouteUsage()
  NamedRerouteUsage()
 ]
}
