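# Material function for the RockChunks skydiving contrail that outputs a vertex-color-masked jiggle vector, apparently for World Position Offset.
# Time times speed, plus the mask times a spread, drives three sines with different periods, forming XYZ motion scaled by amplitude and mask.
# The mask comes from vertex color remapped between min and max parameters, then raised to a power.
# Dynamic Parameter index 3 supplies speed, spread, amplitude and power; MPC_RockChunks parameter nodes remain unconnected.
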MaterialFunction /BRCosmetics/Athena/Effects/Particles/Skydiving/NS_Contrail_RockChunks/Materials/MF_VertColorJiggle
graph {nodes [
  CollectionParameter() {collection=/BRCosmetics/Athena/Effects/Particles/Skydiving/NS_Contrail_RockChunks/MPC_RockChunks, parameterName=1. Frequency}
  CollectionParameter() {collection=&MPC_RockChunks, parameterName="0. Speed(S)"}
  CollectionParameter() {collection=&MPC_RockChunks, parameterName=2. Amplitude}
  CollectionParameter() {collection=&MPC_RockChunks, parameterName=6. VertCol_Pow}
  n1 = DynamicParameter() {parameterIndex=3, defaultValue=<rgba 7 777 2.5 1.2>, paramNames [JIggle_Speed, Jiggle_MaskSpread, Jiggle_Amp, Jiggle_Pow]}
  n2 = DynamicParameter() {parameterIndex=3, defaultValue=<rgba 7 777 2.5 1.2>, paramNames [JIggle_Speed, Jiggle_MaskSpread, Jiggle_Amp, Jiggle_Pow]}
  n3 = Add(A=Multiply(A=Time(), B=@n1.JIggle_Speed), B=Multiply(A=NamedRerouteUsage() {desc=VertColor Mask}, B=@n2.Jiggle_MaskSpread))
  MaterialExpressionAppend3Vector_0 = MaterialXAppend3Vector!(A=Sine(None=@n3) {period=2}, B=Sine(None=@n3) {period=1.3}, C=Sine(None=@n3) {period=2.7})
  n4 = DynamicParameter() {parameterIndex=3, defaultValue=<rgba 7 777 2.5 1.2>, paramNames [JIggle_Speed, Jiggle_MaskSpread, Jiggle_Amp, Jiggle_Pow]}
  FunctionOutput(None=Multiply(A=Multiply(A=@MaterialExpressionAppend3Vector_0, B=@n4.Jiggle_Amp), B=NamedRerouteUsage() {desc=VertColor Mask}))
  n5 = VertexColor()
  n6 = Constant()
  n7 = Constant() {r=1}
  n8 = ScalarParameter() {defaultValue=0.2, group=10. WPO, parameterName=VertCol_Min}
  n9 = ScalarParameter() {defaultValue=1, group=10. WPO, parameterName=VertCol_Max}
  n10 = DynamicParameter() {parameterIndex=3, defaultValue=<rgba 7 777 2.5 1.2>, paramNames [JIggle_Speed, Jiggle_MaskSpread, Jiggle_Amp, Jiggle_Pow]}
  NamedRerouteDeclaration(None=Power(
    Base=MaterialFunctionCall(Input=@n5, Input Low=@n6, Input High=@n7, Target Low=@n8, Target High=@n9) {materialFunction=&RemapValueRange}.Result
    Exp=@n10.Jiggle_Pow
   )) {name=VertColorRemapped, nodeColor=<rgba 0.1484 0 1 1>}
 ]}
