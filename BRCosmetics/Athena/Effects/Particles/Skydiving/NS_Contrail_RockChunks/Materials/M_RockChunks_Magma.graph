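# Lit magma material for mesh particles in the rock chunks skydiving contrail, pairing a parameterized base surface with glowing lava.
# Scaled UVs pan a wispy noise, using LavaSpeed's red channel for both axes, which multiplies a static fire noise.
# The product gets a slight constant tint offset, then emissive color and intensity.
# A static switch, on by default, selects MF_VertColorJiggle or zero, apparently for world position offset.
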Material /BRCosmetics/Athena/Effects/Particles/Skydiving/NS_Contrail_RockChunks/Materials/M_RockChunks_Magma
graph {
 properties {bUsedWithNiagaraMeshParticles=true, pixelDepthOffsetMode=PDOM_Legacy, usedShadingModels=Default Lit}
 nodes [
  n1 = Multiply(A=TextureCoordinate(), B=ScalarParameter() {defaultValue=1, group=3. Magma FX, parameterName=LavaScale})
  n2 = VectorParameter() {group=3. Magma FX, parameterName=LavaSpeed, defaultValue=<rgba 0 0 0 1>}
  n3 = AppendVector(A=@n2.R, B=@n2.R)
  n4 = TextureObjectParameter() {
   group=3. Magma FX
   parameterName=Tex_WispyFireNoise03
   texture=/BRCosmetics/Athena/Effects/Particles/Skydiving/NS_Contrail_Chainmail/T_WispyFireNoise03
  }
  n5 = Multiply(
   A=TextureSample(UVs=Panner(Coordinate=@n1, Speed=@n3) {speedX=1}, Tex=@n4) {texture=&T_WispyFireNoise03}.R
   B=TextureSample(Tex=TextureObjectParameter() {group=3. Magma FX, parameterName=Tex_FireNoise01, texture=/Game/Global/Textures/Default/Noise/T_FireNoise_01}) {texture=&T_FireNoise_01}.R
  )
  n6 = Constant3Vector() {constant=<rgba 0.05 0 0.01 0>}
  Multiply(
   A=Multiply(A=Add(A=@n5, B=@n6), B=VectorParameter() {group=3. Magma FX, parameterName=EmissiveColor, defaultValue=<rgba 1 0.0543 0 0>}.RGB)
   B=ScalarParameter() {defaultValue=1, group=3. Magma FX, parameterName=EmissiveIntensity}
  )
  ScalarParameter() {group=1. Base Material, parameterName=Metallic, sliderMax=1}
  ScalarParameter() {defaultValue=0.5, group=1. Base Material, parameterName=Roughness, sliderMax=1}
  StaticSwitchParameter(
   True=MaterialFunctionCall() {materialFunction=&/MF_VertColorJiggle}.Result
   False=Constant()
  ) {defaultValue=true, group=2. Switch: Use WPO?, parameterName=VertCol_Jiggle?}
  VectorParameter() {group=1. Base Material, parameterName=BaseColor, defaultValue=<rgba 0.0557 0.059 0.042 0>}
 ]
}
