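# Default-lit rock material for Niagara mesh particles in the RockChunks skydiving contrail cosmetic.
# An albedo texture is tinted by a BaseColor parameter, alongside normal and spec-mask textures and scalar Metallic and Roughness controls, all exposed as parameters.
# A static switch, on by default, selects the MF_VertColorJiggle function or zero, apparently for optional vertex-color-based World Position Offset wobble.
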Material /BRCosmetics/Athena/Effects/Particles/Skydiving/NS_Contrail_RockChunks/Materials/M_RockChunks_Rock
graph {
 properties {bUsedWithNiagaraMeshParticles=true, pixelDepthOffsetMode=PDOM_Legacy, usedShadingModels=Default Lit}
 nodes [
  n1 = TextureObjectParameter() {
   group=1. Base Material
   parameterName=Tex_Albedo
   texture=/BRCosmetics/Athena/Effects/Particles/Skydiving/NS_Contrail_RockChunks/Textures/T_Rock02_TP_D_Contrail
  }
  Multiply(
   A=VectorParameter() {group=1. Base Material, parameterName=BaseColor, defaultValue=<rgba 1 1 1 0>}.RGB
   B=TextureSample(Tex=@n1) {samplerType=SAMPLERTYPE_Masks, texture=/Game/Environments/Props/Transylvania/Foliage/Textures/T_TRV_HAF_Rock_M}.RGB
  )
  ScalarParameter() {group=1. Base Material, parameterName=Metallic, sliderMax=1}
  ScalarParameter() {defaultValue=0.5, group=1. Base Material, parameterName=Roughness, sliderMax=1}
  StaticSwitchParameter(
   True=MaterialFunctionCall() {materialFunction=&/MF_VertColorJiggle}.Result
   False=Constant()
  ) {defaultValue=true, group=2. Switch: Use WPO?, parameterName=VertCol_Jiggle?}
  n2 = TextureObjectParameter() {
   group=1. Base Material
   parameterName=Tex_Normals
   samplerType=SAMPLERTYPE_Normal
   texture=/BRCosmetics/Athena/Effects/Particles/Skydiving/NS_Contrail_RockChunks/Textures/T_Rocks_N_Contrail
  }
  TextureSample(Tex=@n2) {samplerType=SAMPLERTYPE_Masks, texture=&T_TRV_HAF_Rock_M}
  n3 = TextureObjectParameter() {
   group=1. Base Material
   parameterName=Tex_SpecMask
   texture=/BRCosmetics/Athena/Effects/Particles/Skydiving/NS_Contrail_RockChunks/Textures/T_Rocks_S_Contrail
  }
  TextureSample(Tex=@n3) {samplerType=SAMPLERTYPE_Masks, texture=&T_TRV_HAF_Rock_M}
 ]
}
